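# Niagara camera lens effect for a SuperSprite dash, built from short burst emitters.
# OuterLensStart drives three mesh renderers from one particle, with separate wind and glow colour, scale and transform attributes plus animated dynamic material parameters.
# SpeedLinesCamera bursts stretched sprites from a sphere, moving along +X with inherited velocity, while WarpRefraction briefly flashes a faint mesh.
# LensEffectLoop and SoftGlow are disabled.
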NiagaraSystem /Sprites_ZeroPointSprite/Assets/FX/CameraLensFX/NS_Camera_SuperSprite_Dash
dependencies {
 usedDataInterfaces [
  &NiagaraDataInterfaceVectorCurve
  &NiagaraDataInterfaceCurve
  &NiagaraDataInterfaceColorCurve
  &NiagaraDataInterfaceVector2DCurve
 ]
 usedDynamicInputs []
 usedModules [&SystemState]
}
emitters {
 LensEffectLoop {
  bEnabled=false
  emitterUpdateScript [
   &EmitterState(Life Cycle Mode=::NewEnumerator1, Loop Behavior=::NewEnumerator1, Loop Duration=1, UseLoopDelay=true)
   &SpawnBurst_Instantaneous(Use Spawn Probability=false, UseLoopCountLimit=false)
  ]
  particleSpawnScript [
   &InitializeParticle(
    Color=<rgba 0.0025 0.2448 0.0039 1>
    Lifetime=0.6
    Mesh Scale=<0.5 0.5 0.5>
    Mesh Scale Mode=::NewEnumerator4
    UsePositionOffset=false
    Write Material Random=true
   )
   /Niagara/Modules/CascadeConversion/CascadeConversion_InitialMeshRotation(Rotation:Vector3f=<0 0 0.5>)
  ]
  particleUpdateScript [
   &ParticleState()
   &ScaleColor(
    Scale Alpha=&FloatFromCurve(FloatCurve=DataInterface() {propertyValues {Curve {keys [<at 0 1>, {arriveTangent=-1, interpMode=RCIM_Cubic, leaveTangent=-1, time=0.5, value=1}, <at 1 0>]}}})
    Scale RGB=&VectorFromFloat(Value=&FloatFromCurve(FloatCurve=DataInterface() {propertyValues {Curve {keys [
          {arriveTangent=-9.2487, interpMode=RCIM_Cubic, leaveTangent=-9.2487, tangentMode=RCTM_User, time=0, value=2}
          {arriveTangent=-2, interpMode=RCIM_Cubic, leaveTangent=-2, time=0.15, value=1}
          {interpMode=RCIM_Cubic, leaveTangent=-3.7472, time=1, value=0}
         ]}}}))
    ScaleA=true
    ScaleRGB=true
   )
  ]
  rendererClasses [&NiagaraMeshRendererProperties]
  renderers [{rendererClass=&NiagaraMeshRendererProperties, rendererIndex=0}]
 }
 OuterLensStart {
  emitterUpdateScript [
   &EmitterState(Life Cycle Mode=::NewEnumerator1, Loop Behavior=::NewEnumerator1, Loop Duration=1, UseLoopDelay=true)
   &SpawnBurst_Instantaneous(Use Spawn Probability=false, UseLoopCountLimit=false)
  ]
  particleSpawnScript [
   &InitializeParticle(
    Lifetime=0.7
    Mesh Scale=&Multiply_VectorByFloat(Vector=<2 1.5 1.5>, Float=2)
    Mesh Scale Mode=::NewEnumerator4
    UsePositionOffset=false
    Write Material Random=true
   )
   &InitialMeshOrientation(Mesh Axis To Orient=<0 0 -1>, Mesh Orientation Mode=::NewEnumerator2, Orientation Vector=<0 0 1>, Use Rotation Vector=false)
   /SetVariables(Particles.MeshOrientationGlow:Quat4f=&MultiplyQuaternion(
     Quaternion A=$Particles.MeshOrientation
     Quaternion B=&OrientMeshToVector(Look At Direction=<1 0 0>, Reference Axis Vector=<1 0 0>)
    )) {isSetParameters=true}
   /SetVariables(Particles.PositionGlow:NiagaraPosition=<-200 0 0>) {isSetParameters=true}
  ]
  particleUpdateScript [
   /SetVariables(Particles.Position=&VectorFromCurve(VectorCurve=DataInterface() {propertyValues {
       XCurve {keys [{arriveTangent=310.7852, interpMode=RCIM_Cubic, time=0, value=-150}, <at 1 0>]}
       YCurve {keys [<at 0 0 RCIM_Cubic>,<at 1 0>]}
       ZCurve {keys [<at 0 0 RCIM_Cubic>,<at 1 0>]}
      }})) {isSetParameters=true}
   &ParticleState()
   &ScaleColor(
    Scale Alpha=&FloatFromCurve(
     FloatCurve=DataInterface() {propertyValues {Curve {keys [<at 0 0 RCIM_Cubic>, <at 0.15 1 RCIM_Cubic>, {interpMode=RCIM_Cubic, leaveTangent=-3.7472, time=1, value=0}]}}}
     Scale Curve=0.45
    )
    Scale RGB=&VectorFromFloat(Value=&FloatFromCurve(FloatCurve=DataInterface() {propertyValues {Curve {keys [
          {arriveTangent=-9.2487, interpMode=RCIM_Cubic, leaveTangent=-9.2487, tangentMode=RCTM_User, time=0, value=1.3}
          {arriveTangent=-1, interpMode=RCIM_Cubic, leaveTangent=-1, time=0.15, value=0.5}
          {interpMode=RCIM_Cubic, leaveTangent=-3.7472, time=1, value=0.3}
         ]}}}))
    ScaleA=true
    ScaleRGB=true
   )
   &DynamicMaterialParameters(
    Index 0 Param 1=&FloatFromCurve(FloatCurve=DataInterface() {propertyValues {Curve {keys [{arriveTangent=-2.4083, interpMode=RCIM_Cubic, time=0, value=0.9}, <at 1 -0.5 RCIM_Cubic>]}}})
    Index 0 Param 2=&FloatFromCurve(FloatCurve=DataInterface() {propertyValues {Curve {keys [
         {arriveTangent=-4.1461, interpMode=RCIM_Cubic, time=0, value=1}
         {arriveTangent=0.2, interpMode=RCIM_Cubic, leaveTangent=0.2, time=0.25, value=0.4}
         {interpMode=RCIM_Cubic, leaveTangent=0.5134, time=1, value=1.2}
        ]}}})
    Index 0 Param 3=4
    Index 0 Param 4=1
    Module.Index0Param0Write=true
    Module.Index0Param1Write=true
    Module.Index0Param2Write=true
    Module.Index0Param3Write=true
   )
   /SetVariables(
    Particles.ColorWind:LinearColor=&ColorFromCurve(ColorCurve=DataInterface() {propertyValues {
       AlphaCurve {keys [<at 0 0 RCIM_Cubic>,<at 0.1022 0.2>,<at 0.8466 0 RCIM_Cubic>]}
       BlueCurve {keys [<at 0.0328 0.24 RCIM_Cubic>,<at 0.3235 0.15>,<at 0.9966 0.03 RCIM_Cubic>]}
       GreenCurve {keys [<at 0.0328 0.768 RCIM_Cubic>,<at 0.3235 0.5042>,<at 0.9966 0.111 RCIM_Cubic>]}
       RedCurve {keys [<at 0.0328 1.2 RCIM_Cubic>,<at 0.3235 1>,<at 0.9966 0.3 RCIM_Cubic>]}
      }})
    Particles.ScaleWind:Vector3f=&Multiply_VectorByFloat(
     Vector=$Particles.Scale
     Float=&Lerp_Float(
      A=0.2
      B=2
      Alpha=&FloatFromCurve(FloatCurve=DataInterface() {propertyValues {Curve {keys [
           {arriveTangent=2.0001, interpMode=RCIM_Cubic, leaveTangent=2.0001, tangentMode=RCTM_User, time=0, value=0}
           {arriveTangent=0.5488, interpMode=RCIM_Cubic, leaveTangent=0.5488, tangentMode=RCTM_User, time=1, value=1}
          ]}}})
     )
    )
   ) {isSetParameters=true}
   /SetVariables(
    Particles.ColorGlow:LinearColor=&ColorFromCurve(ColorCurve=DataInterface() {propertyValues {
       AlphaCurve {keys [<at 0 0 RCIM_Cubic>,<at 0.1078 0.25>,<at 1 0 RCIM_Cubic>]}
       BlueCurve {keys [<at 0.1191 0.05 RCIM_Cubic>,<at 1 0.05 RCIM_Cubic>]}
       GreenCurve {keys [<at 0.1191 0.4933 RCIM_Cubic>,<at 1 0.3667 RCIM_Cubic>]}
       RedCurve {keys [<at 0.1191 1 RCIM_Cubic>,<at 1 1 RCIM_Cubic>]}
      }})
    Particles.ScaleGlow:Vector3f=<0.1 4 4>
   ) {isSetParameters=true}
  ]
  rendererClasses [&NiagaraMeshRendererProperties]
  renderers [
   {rendererClass=&NiagaraMeshRendererProperties, rendererIndex=0}
   {rendererClass=&NiagaraMeshRendererProperties, rendererIndex=1}
   {rendererClass=&NiagaraMeshRendererProperties, rendererIndex=2}
  ]
 }
 SoftGlow {
  bEnabled=false
  emitterUpdateScript [
   &EmitterState(Life Cycle Mode=::NewEnumerator1, Loop Behavior=::NewEnumerator1, Loop Duration=1, UseLoopDelay=true)
   &SpawnBurst_Instantaneous(Use Spawn Probability=false, UseLoopCountLimit=false)
  ]
  particleSpawnScript [&InitializeParticle(
    Color=<rgba 0.0357 0.4271 0 0.25>
    Mesh Scale=<0.01 2.5 2.5>
    Mesh Scale Mode=::NewEnumerator4
    Position=&ConvertVectorToPosition(Input Position=<-200 0 0>, Conversion Operation:ENiagara_LWCConvertVecToPos=::NewEnumerator0)
    Position Mode=::NewEnumerator1
    UsePositionOffset=false
    Write Material Random=true
   )]
  particleUpdateScript [
   &ParticleState()
   &ScaleColor(
    Scale Alpha=&FloatFromCurve(FloatCurve=DataInterface() {propertyValues {Curve {keys [{arriveTangent=29.0114, interpMode=RCIM_Cubic, time=0, value=0}, <at 0.1143 0.8465 RCIM_Cubic>, <at 1 0 RCIM_Cubic>]}}})
    ScaleA=true
    ScaleRGB=true
   )
  ]
  rendererClasses [&NiagaraMeshRendererProperties]
  renderers [{rendererClass=&NiagaraMeshRendererProperties, rendererIndex=0}]
 }
 SpeedLinesCamera {
  emitterUpdateScript [
   &EmitterState(Life Cycle Mode=::NewEnumerator1, Loop Behavior=::NewEnumerator1, Loop Duration=1, UseLoopDelay=true)
   &SpawnBurst_Instantaneous(Spawn Count=45, Use Spawn Probability=false, UseLoopCountLimit=false)
  ]
  particleSpawnScript [
   &InitializeParticle(
    Color=<rgba 0.5 0.26 0.05 1>
    Lifetime=&RandomRangeFloat(Minimum=0.2, Maximum=0.25)
    Sprite Size=&RandomRangeVector2D(Minimum=<2 300>, Maximum=<2 500>)
    Sprite Size Mode=::NewEnumerator4
    UsePositionOffset=false
    Write Material Random=true
   )
   &InheritVelocity(Inherited Velocity Amount Scale=<1500 1500 1500>, Inherited Velocity Speed Limit=0, Speed Limit=false, UseSpeedThreshold=false)
   &ShapeLocation(
    MaskSpawnGroup=false
    Offset=<-50 0 0>
    Offset Mode=::NewEnumerator0
    Override Seed=false
    Rotation Mode=::NewEnumerator1
    Sphere Radius=160
    Sphere Surface Distribution=0.8
    UseOwnerScale=true
   )
   &AddVelocity(Velocity=&Multiply_VectorByFloat(Vector=&Multiply_Vector(A=$Particles.ShapeLocation.ShapeVector, B=<0 1 1>), Float=2)) {module=AddVelocity001}
   &AddVelocity(Velocity=<1600 0 0>)
  ]
  particleUpdateScript [
   &ParticleState()
   &ScaleSpriteSize(
    Scale Factor=&Vector2DFromCurve(
     Vector2Curve=DataInterface() {propertyValues {
       XCurve {keys [
         <at 0.0008 0.9988 RCIM_Cubic>
         {arriveTangent=-0.7133, interpMode=RCIM_Cubic, leaveTangent=-0.7133, time=0.02, value=0.9988}
         {arriveTangent=-1.0192, interpMode=RCIM_Cubic, leaveTangent=-1.0192, time=0.3235, value=0.7686}
         <at 1 0 RCIM_Cubic>
        ]}
       YCurve {keys [
         <at 0.0008 1 RCIM_Cubic>
         {arriveTangent=-2.4731, interpMode=RCIM_Cubic, leaveTangent=-2.4731, time=0.02, value=1}
         {arriveTangent=-1.0204, interpMode=RCIM_Cubic, leaveTangent=-1.0204, time=0.3235, value=0.2018}
         <at 1 0 RCIM_Cubic>
        ]}
      }}
     ScaleCurve=<1 1>
    )
    Scale Sprite Size Mode=::NewEnumerator2
   )
   &SolveForcesAndVelocity(Clamp Velocity=false, Limit Acceleration=false)
   &ScaleColor(
    Scale Alpha=&FloatFromCurve(FloatCurve=DataInterface() {propertyValues {Curve {keys [
         <at 0.1 1 RCIM_Cubic>
         {arriveTangent=-0.698, interpMode=RCIM_Cubic, leaveTangent=-0.698, tangentMode=RCTM_User, time=0.3, value=0.3}
         <at 1 0>
        ]}}})
    ScaleA=true
    ScaleRGB=true
   )
   &DynamicMaterialParameters(
    Index 0 Param 1=0.5
    Index 0 Param 2=0.1
    Index 0 Param 3=1
    Index 0 Param 4=1
    Module.Index0Param0Write=true
    Module.Index0Param1Write=true
    Module.Index0Param2Write=true
    Module.Index0Param3Write=true
   )
  ]
  rendererClasses [&NiagaraSpriteRendererProperties]
  renderers [{rendererClass=&NiagaraSpriteRendererProperties}]
 }
 WarpRefraction {
  emitterUpdateScript [
   &EmitterState(Life Cycle Mode=::NewEnumerator1, Loop Behavior=::NewEnumerator1, Loop Duration=1, UseLoopDelay=true)
   &SpawnBurst_Instantaneous(Use Spawn Probability=false, UseLoopCountLimit=false)
  ]
  particleSpawnScript [&InitializeParticle(
    Color=<rgba 1 1 1 0.1>
    Lifetime=0.3
    Mesh Scale=<1.2 1.2 1.2>
    Mesh Scale Mode=::NewEnumerator4
    Position=&ConvertVectorToPosition(Input Position=<-200 0 0>, Conversion Operation:ENiagara_LWCConvertVecToPos=::NewEnumerator0)
    Position Mode=::NewEnumerator1
    UsePositionOffset=false
    Write Material Random=true
   )]
  particleUpdateScript [
   &ParticleState()
   &ScaleColor(
    Scale Alpha=&FloatFromCurve(FloatCurve=DataInterface() {propertyValues {Curve {keys [{arriveTangent=29.0114, interpMode=RCIM_Cubic, time=0, value=0}, <at 0.1143 0.8465 RCIM_Cubic>, <at 1 0 RCIM_Cubic>]}}})
    Scale RGB=&VectorFromCurve(VectorCurve=DataInterface() {propertyValues {XCurve {keys [<at 0 1>,<at 1 1>]}, YCurve {keys [<at 0 1>,<at 1 1>]}, ZCurve {keys [<at 0 1>,<at 1 1>]}}})
    ScaleA=true
    ScaleRGB=true
   )
  ]
  rendererClasses [&NiagaraMeshRendererProperties]
  renderers [{rendererClass=&NiagaraMeshRendererProperties, rendererIndex=0}]
 }
}
summary {
 systemName=NS_Camera_SuperSprite_Dash
 emitters [
  {emitterName=OuterLensStart, rendererClasses [&NiagaraMeshRendererProperties]}
  {bEnabled=false, emitterName=LensEffectLoop, rendererClasses [&NiagaraMeshRendererProperties]}
  {emitterName=SpeedLinesCamera, rendererClasses [&NiagaraSpriteRendererProperties]}
  {emitterName=WarpRefraction, rendererClasses [&NiagaraMeshRendererProperties]}
  {bEnabled=false, emitterName=SoftGlow, rendererClasses [&NiagaraMeshRendererProperties]}
 ]
 userVariables []
}
